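# Control Rig for the TangyRadishFlame temp female medium body, correcting body deformation and positioning outfit bones.
# Construction imports the skeleton and adds FK controls; forward solve runs shared DeformRig, then Constraints, skirtSetup and HeadUp.
# Weighted parent constraints place ribbon, lace, shirt and skirt base bones, then thigh-up anim attributes drive skirt offsets.
# HeadUp scales neck_02 via a spherical head pose reader, or at full weight when playingEmote is true.
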ControlRigBlueprint /BRCosmetics/Characters/Player/Female/Medium/Bodies/F_Med_TangyRadishFlame_Temp/F_MED_TangyRadishFlame_CtrlRig_TEMP
graphs [
 New Function_ContainedGraph {nodes [
   n1 = RigVMFunctionEntryNode()
   RigVMFunction_Sequence = &RigVMFunction_Sequence(ExecuteContext=@n1.ExecuteContext)
   &RigUnit_ParentConstraint(
    AdvancedSettings=(RotationOrderForFilter=XZY)
    Child=(Type=Bone,Name="base_ribbon_leg_l")
    Parents=((Item=(Type=Bone,Name="pelvis"),Weight=0.0003),(Item=(Type=Bone,Name="deform_knee_out_l"),Weight=0.0155),(Item=(Type=Bone,Name="thigh_twist_01_l"),Weight=0.1048),(Item=(Type=Bone,Name="thigh_l"),Weight=0.8794))
    Weight=1
    ExecutePin=@RigVMFunction_Sequence.A
   )
   > &RigUnit_ParentConstraint(
    AdvancedSettings=(RotationOrderForFilter=XZY)
    Child=(Type=Bone,Name="base_laces_foot_l")
    Parents=((Item=(Type=Bone,Name="calf_twist_01_l"),Weight=0.0002),(Item=(Type=Bone,Name="ball_l"),Weight=0.2172),(Item=(Type=Bone,Name="foot_l"),Weight=0.7826))
    Weight=1
   )
   > &RigUnit_ParentConstraint(
    AdvancedSettings=(RotationOrderForFilter=XZY)
    Child=(Type=Bone,Name="base_front_skirt_l")
    Parents="((Item=(Type=Bone,Name=\"thigh_twist_01_r\"),Weight=0.0005),(Item=(Type=Bone,Name=\"pelvis\"),Weight=0.142),(Item=(Type=Bone,Name=\"thigh_l\"),Weight=0.1752),(Item=(Type=Bone,Name=\"deform_groin_l\"),Weight=0.1855),(Item=(Type=Bone,Name=\"thigh_twist_01_l\"),Weight= ...(truncated)"
    Weight=1
   )
   > &RigUnit_ParentConstraint(
    AdvancedSettings=(RotationOrderForFilter=XZY)
    Child=(Type=Bone,Name="base_side_skirt_l")
    Parents="((Item=(Type=Bone,Name=\"deform_groin_l\"),Weight=0.0151),(Item=(Type=Bone,Name=\"deform_glute_l\"),Weight=0.0339),(Item=(Type=Bone,Name=\"thigh_l\"),Weight=0.1334),(Item=(Type=Bone,Name=\"pelvis\"),Weight=0.1466),(Item=(Type=Bone,Name=\"thigh_twist_01_l\"),Weight=0 ...(truncated)"
    Weight=1
   )
   > &RigUnit_ParentConstraint(
    AdvancedSettings=(RotationOrderForFilter=XZY)
    Child=(Type=Bone,Name="base_bk_skirt_l")
    Parents=((Item=(Type=Bone,Name="thigh_l"),Weight=0.0001),(Item=(Type=Bone,Name="thigh_twist_01_l"),Weight=0.1735),(Item=(Type=Bone,Name="pelvis"),Weight=0.3926),(Item=(Type=Bone,Name="deform_glute_l"),Weight=0.4337))
    Weight=1
   )
   > &RigUnit_ParentConstraint(
    AdvancedSettings=(RotationOrderForFilter=XZY)
    Child=(Type=Bone,Name="base_ribbon_chest_B")
    Parents=((Item=(Type=Bone,Name="spine_02"),Weight=0.036),(Item=(Type=Bone,Name="spine_03"),Weight=0.0362),(Item=(Type=Bone,Name="spine_05"),Weight=0.0489),(Item=(Type=Bone,Name="spine_04"),Weight=0.8789))
    Weight=1
   )
   > &RigUnit_ParentConstraint(
    AdvancedSettings=(RotationOrderForFilter=XZY)
    Child=(Type=Bone,Name="base_ribbon_chest_A")
    Parents=((Item=(Type=Bone,Name="spine_05"),Weight=0.2448),(Item=(Type=Bone,Name="spine_04"),Weight=0.7552))
    Weight=1
   )
   > &RigUnit_ParentConstraint(
    AdvancedSettings=(RotationOrderForFilter=XZY)
    Child=(Type=Bone,Name="base_side_shirt_l")
    Parents="((Item=(Type=Bone,Name=\"upperarm_twist_01_l\"),Weight=0.0118),(Item=(Type=Bone,Name=\"pelvis\"),Weight=0.0298),(Item=(Type=Bone,Name=\"clavicle_l\"),Weight=0.0302),(Item=(Type=Bone,Name=\"spine_05\"),Weight=0.0344),(Item=(Type=Bone,Name=\"spine_01\"),Weight=0.0458) ...(truncated)"
    Weight=1
   )
   > &RigUnit_ParentConstraint(
    AdvancedSettings=(RotationOrderForFilter=XZY)
    Child=(Type=Bone,Name="base_bk_shirt_l")
    Parents="((Item=(Type=Bone,Name=\"upperarm_twist_01_l\"),Weight=0.0033),(Item=(Type=Bone,Name=\"clavicle_l\"),Weight=0.0419),(Item=(Type=Bone,Name=\"spine_05\"),Weight=0.0433),(Item=(Type=Bone,Name=\"pelvis\"),Weight=0.0486),(Item=(Type=Bone,Name=\"spine_01\"),Weight=0.0798) ...(truncated)"
    Weight=1
   )
   > &RigUnit_ParentConstraint(
    AdvancedSettings=(RotationOrderForFilter=XZY)
    Child=(Type=Bone,Name="base_front_shirt_r")
    Parents=((Item=(Type=Bone,Name="pelvis"),Weight=0.0108),(Item=(Type=Bone,Name="spine_01"),Weight=0.0237),(Item=(Type=Bone,Name="spine_03"),Weight=0.1066),(Item=(Type=Bone,Name="spine_02"),Weight=0.1544),(Item=(Type=Bone,Name="spine_04"),Weight=0.7044))
    Weight=1
   )
   > &RigUnit_ParentConstraint(
    AdvancedSettings=(RotationOrderForFilter=XZY)
    Child=(Type=Bone,Name="base_front_shirt_l")
    Parents=((Item=(Type=Bone,Name="pelvis"),Weight=0.0108),(Item=(Type=Bone,Name="spine_01"),Weight=0.0237),(Item=(Type=Bone,Name="spine_03"),Weight=0.1066),(Item=(Type=Bone,Name="spine_02"),Weight=0.1544),(Item=(Type=Bone,Name="spine_04"),Weight=0.7044))
    Weight=1
   )
   > &RigUnit_ParentConstraint(
    AdvancedSettings=(RotationOrderForFilter=XZY)
    Child=(Type=Bone,Name="base_bk_shirt_r")
    Parents="((Item=(Type=Bone,Name=\"upperarm_twist_01_r\"),Weight=0.0033),(Item=(Type=Bone,Name=\"clavicle_r\"),Weight=0.0419),(Item=(Type=Bone,Name=\"spine_05\"),Weight=0.0433),(Item=(Type=Bone,Name=\"pelvis\"),Weight=0.0486),(Item=(Type=Bone,Name=\"spine_01\"),Weight=0.0798) ...(truncated)"
    Weight=1
   )
   > &RigUnit_ParentConstraint(
    AdvancedSettings=(RotationOrderForFilter=XZY)
    Child=(Type=Bone,Name="base_side_shirt_r")
    Parents="((Item=(Type=Bone,Name=\"upperarm_twist_01_r\"),Weight=0.0161),(Item=(Type=Bone,Name=\"pelvis\"),Weight=0.0241),(Item=(Type=Bone,Name=\"clavicle_r\"),Weight=0.0354),(Item=(Type=Bone,Name=\"spine_05\"),Weight=0.0404),(Item=(Type=Bone,Name=\"spine_01\"),Weight=0.0423) ...(truncated)"
    Weight=1
   )
   > &RigUnit_ParentConstraint(
    AdvancedSettings=(RotationOrderForFilter=XZY)
    Child=(Type=Bone,Name="base_ribbon_arm_r")
    Parents=((Item=(Type=Bone,Name="upperarm_twist_02_r"),Weight=0.1638),(Item=(Type=Bone,Name="upperarm_r"),Weight=0.1646),(Item=(Type=Bone,Name="upperarm_twist_01_r"),Weight=0.6716))
    Weight=1
   )
   > &RigUnit_ParentConstraint(
    AdvancedSettings=(RotationOrderForFilter=XZY)
    Child=(Type=Bone,Name="base_laces_foot_r")
    Parents=((Item=(Type=Bone,Name="calf_twist_01_r"),Weight=0.0002),(Item=(Type=Bone,Name="ball_r"),Weight=0.2172),(Item=(Type=Bone,Name="foot_r"),Weight=0.7826))
    Weight=1
   )
   > &RigUnit_ParentConstraint(
    AdvancedSettings=(RotationOrderForFilter=XZY)
    Child=(Type=Bone,Name="base_front_skirt_r")
    Parents="((Item=(Type=Bone,Name=\"thigh_twist_01_l\"),Weight=0.0005),(Item=(Type=Bone,Name=\"pelvis\"),Weight=0.1422),(Item=(Type=Bone,Name=\"thigh_r\"),Weight=0.1753),(Item=(Type=Bone,Name=\"deform_groin_r\"),Weight=0.1853),(Item=(Type=Bone,Name=\"thigh_twist_01_r\"),Weight ...(truncated)"
    Weight=1
   )
   > &RigUnit_ParentConstraint(
    AdvancedSettings=(RotationOrderForFilter=XZY)
    Child=(Type=Bone,Name="base_side_skirt_r")
    Parents="((Item=(Type=Bone,Name=\"deform_groin_r\"),Weight=0.0137),(Item=(Type=Bone,Name=\"deform_glute_r\"),Weight=0.0324),(Item=(Type=Bone,Name=\"thigh_r\"),Weight=0.1317),(Item=(Type=Bone,Name=\"pelvis\"),Weight=0.1445),(Item=(Type=Bone,Name=\"thigh_twist_01_r\"),Weight=0 ...(truncated)"
    Weight=1
   )
   > &RigUnit_ParentConstraint(
    AdvancedSettings=(RotationOrderForFilter=XZY)
    Child=(Type=Bone,Name="base_bk_skirt_r")
    Parents="((Item=(Type=Bone,Name=\"deform_groin_r\"),Weight=0.0002),(Item=(Type=Bone,Name=\"thigh_l\"),Weight=0.0003),(Item=(Type=Bone,Name=\"deform_glute_l\"),Weight=0.0045),(Item=(Type=Bone,Name=\"thigh_twist_01_l\"),Weight=0.0054),(Item=(Type=Bone,Name=\"thigh_r\"),Weight= ...(truncated)"
    Weight=1
   )
   > &RigUnit_ParentConstraint(
    AdvancedSettings=(RotationOrderForFilter=XZY)
    Child=(Type=Bone,Name="base_ribbon_pelvis")
    Parents="((Item=(Type=Bone,Name=\"thigh_twist_01_r\"),Weight=0.0001),(Item=(Type=Bone,Name=\"thigh_l\"),Weight=0.0003),(Item=(Type=Bone,Name=\"spine_02\"),Weight=0.0396),(Item=(Type=Bone,Name=\"thigh_twist_01_l\"),Weight=0.0892),(Item=(Type=Bone,Name=\"deform_groin_l\"),Weig ...(truncated)"
    Weight=1
   )
   RigVMFunctionReturnNode(ExecuteContext=@RigVMFunction_Sequence.B)
   &RigUnit_ParentConstraint(
    AdvancedSettings=(RotationOrderForFilter=XZY)
    Child=(Type=Bone,Name="base_ribbon_pelvis")
    ExecutePin=()
    Parents=((Item=(Type=Bone,Name="thigh_twist_01_l"),Weight=0.0221),(Item=(Type=Bone,Name="spine_02"),Weight=0.0608),(Item=(Type=Bone,Name="deform_groin_l"),Weight=0.067),(Item=(Type=Bone,Name="pelvis"),Weight=0.85))
    Weight=1
   )
  ]}
 New Function_ContainedGraph {nodes [
   n1 = RigVMFunctionEntryNode()
   RigVMFunction_Sequence = &RigVMFunction_Sequence(ExecuteContext=@n1.ExecuteContext)
   &RigUnit_ModifyTransforms(
    ItemToModify="((Item=(Type=Bone,Name=\"base_front_skirt_l\"),Transform=(Rotation=(X=0,Y=-0.2164,Z=0,W=0.9763),Translation=(X=0,Y=0,Z=0),Scale3D=(X=1,Y=1,Z=1))),(Item=(Type=Bone,Name=\"base_side_skirt_l\"),Transform=(Rotation=(X=0,Y=0,Z=0,W=1),Translation=(X=6,Y=0,Z=0),Scale ...(truncated)"
    WeightMinimum=0
    Weight=RigVMDispatchNode(BoneName=root, Default=0, Name=deformRig_l_thigh_up_alpha) {template="DISPATCH_RigDispatch_GetAnimAttribute(in Name,in BoneName,in Default,out Value,out Found)"}.Value
    ExecutePin=@RigVMFunction_Sequence.A
   )
   > &RigUnit_ModifyTransforms(
    ItemToModify="((Item=(Type=Bone,Name=\"base_front_skirt_r\"),Transform=(Rotation=(X=0,Y=-0.2164,Z=0,W=0.9763),Translation=(X=0,Y=0,Z=0),Scale3D=(X=1,Y=1,Z=1))),(Item=(Type=Bone,Name=\"base_side_skirt_r\"),Transform=(Rotation=(X=0,Y=0,Z=0,W=1),Translation=(X=0,Y=0,Z=-0.5),Sc ...(truncated)"
    WeightMinimum=0
    Weight=RigVMDispatchNode(BoneName=root, Default=0, Name=deformRig_r_thigh_up_alpha) {template="DISPATCH_RigDispatch_GetAnimAttribute(in Name,in BoneName,in Default,out Value,out Found)"}.Value
   )
   RigVMFunctionReturnNode(ExecuteContext=@RigVMFunction_Sequence.B)
  ]}
 New Function_ContainedGraph {nodes [
   n1 = RigVMFunctionEntryNode()
   RigVMFunction_Sequence = &RigVMFunction_Sequence(ExecuteContext=@n1.ExecuteContext)
   RigVMFunction_ControlFlowBranch = &RigVMFunction_ControlFlowBranch(ExecuteContext=@RigVMFunction_Sequence.A, Condition=RigVMVariableNode() {variable=playingEmote}.Value)
   n2 = &RigUnit_SphericalPoseReader(
    ActiveRegionScaleFactors=(PositiveWidth=1,NegativeWidth=0.2,PositiveHeight=1.5,NegativeHeight=1.5)
    ActiveRegionSize=0.1
    Debug=(bDrawDebug=false,bDraw2D=false,bDrawLocalAxes=false,DebugScale=25,DebugSegments=20,DebugThickness=0.25)
    DriverAxis=(X=0,Y=1,Z=0)
    DriverItem=(Type=Bone,Name="head")
    FalloffRegionScaleFactors=(PositiveWidth=3,NegativeWidth=1.8,PositiveHeight=2.5,NegativeHeight=2.5)
    FalloffSize=0.2
    FlipHeightScaling=False
    FlipWidthScaling=False
    OptionalParentItem=(Type=Bone,Name="None")
    RotationOffset=(X=90,Y=0,Z=-55)
    ExecutePin=@RigVMFunction_ControlFlowBranch.False
   )
   Multiply = &RigVMFunction_MathDoubleMul(B=2, A=@n2.OutputParam)
   > &RigUnit_ModifyTransforms(
    ItemToModify=((Item=(Type=Bone,Name="neck_02"),Transform=(Rotation=(X=0,Y=0,Z=0,W=1),Translation=(X=0,Y=0,Z=0),Scale3D=(X=1,Y=1.5,Z=1.4))))
    WeightMinimum=0
    Weight=@Multiply.Result
   )
   &RigUnit_ModifyTransforms(
    ItemToModify=((Item=(Type=Bone,Name="neck_02"),Transform=(Rotation=(X=0,Y=0,Z=0,W=1),Translation=(X=0,Y=1,Z=0),Scale3D=(X=1,Y=1.4,Z=1.4))))
    Weight=1
    WeightMinimum=0
    ExecutePin=@RigVMFunction_ControlFlowBranch.True
   )
   RigVMFunctionReturnNode(ExecuteContext=@RigVMFunction_Sequence.B)
  ]}
 RigVMFunctionLibrary {nodes [
   Constraints = RigVMCollapseNode(ExecuteContext=())
   HeadUp = RigVMCollapseNode(ExecuteContext=())
   skirtSetup = RigVMCollapseNode(ExecuteContext=())
  ]}
 RigVMModel {nodes [
   n1 = &RigUnit_PrepareForExecution()
   &RigUnit_HierarchyImportFromSkeleton(NameSpace=None, bIncludeCurves=False, bIncludeMeshSockets=False, bIncludeVirtualBones=True, ExecutePin=@n1.ExecutePin)
   > addFKCtrl = RigVMFunctionReferenceNode(
    Custom_FKCtrl=()
    Options_FKCtrl="(pelvis_1_#1=true,spine_3_#2=true,clavicle_5_#3=true,arm_7_#4=true,leg_9_#5=true,head_11_46DC15EF465CD2979 ...(truncated)"
    Shape_Settings="(pelvis_2_#6=(bVisible=True,Name=\"Hexagon_Thin\",Color=(R=0.0117,G=0.8281,B=1,A=1),Transform=(Rotation=(X=0,Y=-0.7071,Z=0,W=0.7071),Translation=(X=2,Y=0,Z=0),Scale3D=(X=5,Y=5,Z=7))),spine_4_#7=(bVi ...(truncated)"
   )
   RigUnit_BeginExecution = &RigUnit_BeginExecution()
   FN_DeformRig_FNC_1 = RigVMFunctionReferenceNode(
    DeformRig_ Curve="(deformRig_l_upperarm_up_alpha_72_#8=False,deformRig_r_upperarm_up_alpha_73_#9=False,deformRig_l_upperarm_fr_alpha_74_#10=False,deformRig_r_upperarm_fr_alpha_75_CB51B8 ...(truncated)"
    DeformRig_Bool="(upperarm_1_#11=True,elbow_3_#12=True,wrist_7_#13=True,groin_8_#14=True,glute_17_#15=True,knee_9_40C408324CA261 ...(truncated)"
    DeformRig_Debug=(upperarm_debug_18_#11=False,wrist_debug_20_#13=False,lat_debug_26_#16=False,groinGlute_debug_27_#14=False)
    DeformRig_Value="(upperarm_up_18_#17=(Rotation=(X=0,Y=0,Z=0,W=1),Translation=(X=0,Y=0,Z=8),Scale3D=(X=0.7,Y=1,Z=1)),upperarm_fwd_fr_19_#18=(Rotation=(X=0,Y=0,Z=0,W=1),Translation=(X=0,Y=10,Z=0),Scale3D=(X=0.5,Y=1,Z= ...(truncated)"
    ExecuteContext=@RigUnit_BeginExecution.ExecutePin
   )
   > Constraints = RigVMFunctionReferenceNode()
   > skirtSetup = RigVMFunctionReferenceNode()
   > HeadUp = RigVMFunctionReferenceNode()
   setFKCtrl = RigVMFunctionReferenceNode(ExecuteContext=(), setFK_Ctrl=false)
  ]}
]
hierarchy [
 root {loc=<0 0 0>, rot=<0 0 0 1>}
 pelvis {parent=root, loc=<0 -0.5301 98.0571>, rot=<0.0127 0.707 -0.0127 -0.707>}
 spine_01 {parent=pelvis, loc=<6.1208 -0.9328 0>, rot=<0 0 0.034 -0.9994>}
 spine_02 {parent=spine_01, loc=<6.5997 -0.2095 0>, rot=<0 0 0.0238 -0.9997>}
 spine_03 {parent=spine_02, loc=<6.4914 -0.1237 0>, rot=<0 0 0.0149 -0.9999>}
 spine_04 {parent=spine_03, loc=<6.9598 0.0794 0>, rot=<0 0 -0.0132 -0.9999>}
 spine_05 {parent=spine_04, loc=<6.9987 0 0>, rot=<0 0 -0.0015 -1>}
 neck_01 {parent=spine_05, loc=<11.3138 0.1529 0>, rot=<0 0 -0.2003 -0.9797>}
 neck_02 {parent=neck_01, loc=<4.5627 0 0>, rot=<0 0 0.0504 -0.9987>}
 head {parent=neck_02, loc=<5.3914 0 0>, rot=<0 0 0.115 -0.9934>}
 clavicle_l {parent=spine_05, loc=<5.4831 -0.1053 -3.2946>, rot=<0.0465 -0.7289 -0.009 -0.6829>}
 upperarm_l {parent=clavicle_l, loc=<9.8658 -1.2136 -0.4105>, rot=<0.0164 -0.3245 -0.0269 -0.9454>}
 lowerarm_l {parent=upperarm_l, loc=<26.465 -0.8345 0.259>, rot=<0.0036 0.0161 -0.2298 -0.9731>}
 hand_l {parent=lowerarm_l, loc=<22.7355 -0.3203 -0.6105>, rot=<-0.6784 -0.0223 0.0735 -0.7307>}
 thumb_01_l {parent=hand_l, loc=<1.3118 -0.3191 -2.459>, rot=<0.6546 -0.4042 0.0664 -0.6354>}
 thumb_02_l {parent=thumb_01_l, loc=<5.9336 -0.0448 0.0401>, rot=<-0.0037 0.0213 0.298 -0.9543>}
 thumb_03_l {parent=thumb_02_l, loc=<2.2223 0.3682 -0.0907>, rot=<0.0118 -0.013 0.0066 -0.9998>}
 index_metacarpal_l {parent=hand_l, loc=<1.6535 0.0183 -2.96>, rot=<0.0158 -0.0124 -0.0173 -0.9996>}
 index_01_l {parent=index_metacarpal_l, loc=<8.3008 0.0357 -0.0524>, rot=<-0.0044 0.0485 0.0924 -0.9945>}
 index_02_l {parent=index_01_l, loc=<3.3587 -0.3004 0.0452>, rot=<-0.002 0.0027 0.3225 -0.9466>}
 index_03_l {parent=index_02_l, loc=<2.5578 -0.123 -0.0252>, rot=<-0.002 0.0067 0.2558 -0.9667>}
 middle_metacarpal_l {parent=hand_l, loc=<2.1856 -0.1725 -0.5984>, rot=<0.006 0.0168 -0.0329 -0.9993>}
 middle_01_l {parent=middle_metacarpal_l, loc=<7.3275 -0.2447 0.0129>, rot=<-0.0208 0.0042 0.0858 -0.9961>}
 middle_02_l {parent=middle_01_l, loc=<4.0123 0.1401 0.1175>, rot=<-0.0218 0.0258 0.3512 -0.9357>}
 middle_03_l {parent=middle_02_l, loc=<2.6077 0.0534 0.1373>, rot=<0.0082 0.0188 0.1708 -0.9851>}
 ring_metacarpal_l {parent=hand_l, loc=<2.2074 -0.3625 1.3638>, rot=<-0.0284 0.0722 -0.0458 -0.9959>}
 ring_01_l {parent=ring_metacarpal_l, loc=<7.0295 -0.6197 -0.1446>, rot=<-0.0489 -0.0252 0.1892 -0.9804>}
 ring_02_l {parent=ring_01_l, loc=<3.6174 0.4647 -0.1027>, rot=<0 0 0.2806 -0.9598>}
 ring_03_l {parent=ring_02_l, loc=<2.4152 0.114 0.0199>, rot=<0 0 0.1189 -0.9929>}
 pinky_metacarpal_l {parent=hand_l, loc=<1.7616 -0.9753 3.3814>, rot=<-0.0397 0.0814 -0.069 -0.9935>}
 pinky_01_l {parent=pinky_metacarpal_l, loc=<7.1709 -0.4831 -0.1094>, rot=<-0.0654 -0.0035 0.2242 -0.9723>}
 pinky_02_l {parent=pinky_01_l, loc=<2.8274 0.0893 -0.1043>, rot=<0.0037 0.0011 0.2854 -0.9584>}
 pinky_03_l {parent=pinky_02_l, loc=<2.2046 0.0307 -0.0078>, rot=<-0.0002 0.0098 0.1445 -0.9895>}
 deform_wrist_dn_base_l {parent=hand_l, loc=<0 -1.5326 0>, rot=<0 0 0 -1>}
 deform_wrist_dn_l {parent=deform_wrist_dn_base_l, loc=<0 0 0>, rot=<0 0 0 -1>}
 deform_wrist_up_base_l {parent=hand_l, loc=<0 2.3009 0>, rot=<0 0 0 -1>}
 deform_wrist_up_l {parent=deform_wrist_up_base_l, loc=<0 0 0>, rot=<0 0 0 -1>}
 lowerarm_twist_02_l {parent=lowerarm_l, loc=<10.6707 0 0>, rot=<0 0 0 1>}
 lowerarm_twist_01_l {parent=lowerarm_l, loc=<18.316 0 0>, rot=<0 0 0 1>}
 deform_elbow_in_base_l {parent=lowerarm_l, loc=<0.1834 3.6062 -0.2572>, rot=<0 0 0 -1>}
 deform_elbow_in_l {parent=deform_elbow_in_base_l, loc=<0 0 0>, rot=<0 0 0 -1>}
 deform_elbow_out_base_l {parent=lowerarm_l, loc=<0 -4 0>, rot=<0 0 0 1>}
 deform_elbow_out_l {parent=deform_elbow_out_base_l, loc=<0 0 0>, rot=<0 0 0 1>}
 upperarm_twist_01_l {parent=upperarm_l, loc=<7.2135 -0.1134 0>, rot=<0 0 0 1>}
 deform_upperarm_bk_base_l {parent=upperarm_twist_01_l, loc=<-6.4053 -4.3746 0.314>, rot=<0 0 0 -1>}
 deform_upperarm_bk_l {parent=deform_upperarm_bk_base_l, loc=<0 0 0>, rot=<0 0 0 -1>}
 deform_upperarm_fr_base_l {parent=upperarm_twist_01_l, loc=<-5.6312 5.092 0.3354>, rot=<0 0 0 -1>}
 deform_upperarm_fr_l {parent=deform_upperarm_fr_base_l, loc=<0 0 0>, rot=<0 0 0 -1>}
 deform_upperarm_up_base_l {parent=upperarm_twist_01_l, loc=<-8.891 0.4825 3.9517>, rot=<0 0 0 -1>}
 deform_upperarm_up_l {parent=deform_upperarm_up_base_l, loc=<0 0 0>, rot=<0 0 0 -1>}
 upperarm_twist_02_l {parent=upperarm_l, loc=<19.7998 -0.1134 0>, rot=<0 0 0 1>}
 clavicle_r {parent=spine_05, loc=<5.4831 -0.1053 3.2946>, rot=<0.7289 0.0465 -0.6829 0.009>}
 upperarm_r {parent=clavicle_r, loc=<-9.8658 1.2136 0.4105>, rot=<0.0164 -0.3245 -0.0269 -0.9454>}
 lowerarm_r {parent=upperarm_r, loc=<-26.465 0.8345 -0.259>, rot=<0.0036 0.0161 -0.2298 -0.9731>}
 hand_r {parent=lowerarm_r, loc=<-22.7355 0.3203 0.6106>, rot=<-0.6784 -0.0223 0.0735 -0.7307>}
 thumb_01_r {parent=hand_r, loc=<-1.3118 0.3191 2.459>, rot=<0.6546 -0.4042 0.0664 -0.6354>}
 thumb_02_r {parent=thumb_01_r, loc=<-5.9336 0.0448 -0.0401>, rot=<-0.0037 0.0213 0.298 -0.9543>}
 thumb_03_r {parent=thumb_02_r, loc=<-2.2223 -0.3682 0.0907>, rot=<0.0118 -0.013 0.0066 -0.9998>}
 index_metacarpal_r {parent=hand_r, loc=<-1.6535 -0.0183 2.96>, rot=<0.0158 -0.0124 -0.0173 -0.9996>}
 index_01_r {parent=index_metacarpal_r, loc=<-8.3008 -0.0357 0.0524>, rot=<-0.0044 0.0485 0.0924 -0.9945>}
 index_02_r {parent=index_01_r, loc=<-3.3587 0.3004 -0.0452>, rot=<-0.002 0.0027 0.3225 -0.9466>}
 index_03_r {parent=index_02_r, loc=<-2.5578 0.123 0.0252>, rot=<-0.002 0.0067 0.2558 -0.9667>}
 middle_metacarpal_r {parent=hand_r, loc=<-2.1856 0.1725 0.5984>, rot=<0.006 0.0168 -0.0329 -0.9993>}
 middle_01_r {parent=middle_metacarpal_r, loc=<-7.3275 0.2447 -0.0129>, rot=<-0.0208 0.0042 0.0858 -0.9961>}
 middle_02_r {parent=middle_01_r, loc=<-4.0123 -0.1401 -0.1175>, rot=<-0.0218 0.0258 0.3512 -0.9357>}
 middle_03_r {parent=middle_02_r, loc=<-2.6077 -0.0534 -0.1373>, rot=<0.0082 0.0188 0.1708 -0.9851>}
 ring_metacarpal_r {parent=hand_r, loc=<-2.2074 0.3625 -1.3638>, rot=<-0.0284 0.0722 -0.0458 -0.9959>}
 ring_01_r {parent=ring_metacarpal_r, loc=<-7.0295 0.6197 0.1446>, rot=<-0.0489 -0.0252 0.1892 -0.9804>}
 ring_02_r {parent=ring_01_r, loc=<-3.6174 -0.4647 0.1027>, rot=<0 0 0.2806 -0.9598>}
 ring_03_r {parent=ring_02_r, loc=<-2.4152 -0.114 -0.0199>, rot=<0 0 0.1189 -0.9929>}
 pinky_metacarpal_r {parent=hand_r, loc=<-1.7616 0.9753 -3.3814>, rot=<-0.0397 0.0814 -0.069 -0.9935>}
 pinky_01_r {parent=pinky_metacarpal_r, loc=<-7.1709 0.4831 0.1094>, rot=<-0.0654 -0.0035 0.2242 -0.9723>}
 pinky_02_r {parent=pinky_01_r, loc=<-2.8274 -0.0893 0.1043>, rot=<0.0037 0.0011 0.2854 -0.9584>}
 pinky_03_r {parent=pinky_02_r, loc=<-2.2046 -0.0307 0.0078>, rot=<-0.0002 0.0098 0.1445 -0.9895>}
 deform_wrist_dn_base_r {parent=hand_r, loc=<-0.0001 1.5328 0>, rot=<0 0 0 -1>}
 deform_wrist_dn_r {parent=deform_wrist_dn_base_r, loc=<0 0 0>, rot=<0 0 0 -1>}
 deform_wrist_up_base_r {parent=hand_r, loc=<-0.0002 -2.3007 -0.0001>, rot=<0 0 0 -1>}
 deform_wrist_up_r {parent=deform_wrist_up_base_r, loc=<0 0 0>, rot=<0 0 0 -1>}
 lowerarm_twist_02_r {parent=lowerarm_r, loc=<-10.6707 0 0>, rot=<0 0 0 -1>}
 lowerarm_twist_01_r {parent=lowerarm_r, loc=<-18.316 0 0>, rot=<0 0 0 -1>}
 deform_elbow_in_base_r {parent=lowerarm_r, loc=<-0.1834 -3.6062 0.2572>, rot=<0 0 0 -1>}
 deform_elbow_in_r {parent=deform_elbow_in_base_r, loc=<0 0 0>, rot=<0 0 0 -1>}
 deform_elbow_out_base_r {parent=lowerarm_r, loc=<0 4 -0.0001>, rot=<0 0 0 -1>}
 deform_elbow_out_r {parent=deform_elbow_out_base_r, loc=<0 0 0>, rot=<0 0 0 1>}
 upperarm_twist_01_r {parent=upperarm_r, loc=<-7.2135 0.1134 0>, rot=<0 0 0 -1>}
 deform_upperarm_bk_base_r {parent=upperarm_twist_01_r, loc=<6.4053 4.3746 -0.314>, rot=<0 0 0 -1>}
 deform_upperarm_bk_r {parent=deform_upperarm_bk_base_r, loc=<0 0 0>, rot=<0 0 0 -1>}
 deform_upperarm_fr_base_r {parent=upperarm_twist_01_r, loc=<5.6312 -5.092 -0.3354>, rot=<0 0 0 -1>}
 deform_upperarm_fr_r {parent=deform_upperarm_fr_base_r, loc=<0 0 0>, rot=<0 0 0 -1>}
 deform_upperarm_up_base_r {parent=upperarm_twist_01_r, loc=<8.891 -0.4825 -3.9517>, rot=<0 0 0 -1>}
 deform_upperarm_up_r {parent=deform_upperarm_up_base_r, loc=<0 0 0>, rot=<0 0 0 -1>}
 base_ribbon_arm_r {parent=upperarm_twist_01_r, loc=<-6.2737 -0.8946 -2.687>, rot=<0.7049 0.695 -0.1168 0.0804>}
 dyn_main_ribbon_arm_r {parent=base_ribbon_arm_r, loc=<0 0 0>, rot=<0 0 0 -1>}
 base_ribbon_tape_arm_r {parent=dyn_main_ribbon_arm_r, loc=<0.1412 -1.0255 0.0343>, rot=<0 0 0.7071 -0.7071>}
 dyn_ribbon_tape_arm_r {parent=base_ribbon_tape_arm_r, loc=<0 0 0>, rot=<0 0 0 1>}
 dyn_ribbon_arm_r_r {parent=dyn_main_ribbon_arm_r, loc=<0.8146 0 0>, rot=<0 0 0 -1>}
 dyn_ribbon_arm_r_l {parent=dyn_main_ribbon_arm_r, loc=<-0.669 0 0>, rot=<0 0 0 -1>}
 upperarm_twist_02_r {parent=upperarm_r, loc=<-19.7998 0.1134 0>, rot=<0 0 0 -1>}
 attach_backpack {parent=spine_05, loc=<-0.123 -13.473 0>, rot=<-0.0093 -0.707 0.0093 -0.707>}
 Backpack_BR {parent=attach_backpack, type=Null, loc=<0 3.0606 0>, rot=<0 0 0 1>}
 attach_cape {parent=spine_05, loc=<9.017 -11.714 0>, rot=<0.0616 -0.7044 -0.0616 -0.7044>}
 cape {parent=attach_cape, type=Null, loc=<0 1.527 0.1336>, rot=<0.0436 0 0 0.999>}
 deform_pec_aim_base_l {parent=spine_04, loc=<3.2456 9.3309 -1.2237>, rot=<0.0603 -0.4033 0.3318 -0.8507>}
 deform_pec_aim_l {parent=deform_pec_aim_base_l, loc=<0 0 0>, rot=<0 0 0 -1>}
 deform_pec_base_l {parent=deform_pec_aim_l, loc=<9.3665 0 0>, rot=<0 0 0 -1>}
 deform_pec_l {parent=deform_pec_base_l, loc=<0 0 0>, rot=<0 0 0 -1>}
 deform_pec_aim_base_r {parent=spine_04, loc=<3.2452 9.3309 1.2237>, rot=<-0.4033 -0.0603 0.8507 0.3318>}
 deform_pec_aim_r {parent=deform_pec_aim_base_r, loc=<0 0 0>, rot=<0 0 0 -1>}
 deform_pec_base_r {parent=deform_pec_aim_r, loc=<-9.3667 -0.0003 -0.0004>, rot=<0 0 0 -1>}
 deform_pec_r {parent=deform_pec_base_r, loc=<0 0 0>, rot=<0 0 0 -1>}
 base_ribbon_chest_B {parent=spine_04, loc=<-2.1387 14.5611 0>, rot=<-0.0294 0.7065 0.7065 -0.0294>}
 dyn_ribbon_main_chest_B {parent=base_ribbon_chest_B, loc=<0 0 0>, rot=<0 0 0 1>}
 dyn_ribbon_chest_B_l {parent=dyn_ribbon_main_chest_B, loc=<0 -0.8385 0>, rot=<0 0 0.7071 -0.7071>}
 dyn_ribbon_chest_B_r {parent=dyn_ribbon_main_chest_B, loc=<0 0.8352 0>, rot=<0 0 -0.7071 -0.7071>}
 base_ribbon_tape_B {parent=dyn_ribbon_main_chest_B, loc=<1.0202 0 0>, rot=<0 0.0556 0 -0.9985>}
 dyn_ribbon_tape_chest_B_r_1 {parent=base_ribbon_tape_B, loc=<0.2658 0.7697 -0.8225>, rot=<0.0122 -0.0707 -0.0277 -0.997>}
 dyn_ribbon_tape_chest_B_r_2 {parent=dyn_ribbon_tape_chest_B_r_1, loc=<2.6091 0 0>, rot=<0 0 0 1>}
 dyn_ribbon_tape_chest_B_l_1 {parent=base_ribbon_tape_B, loc=<0.27 -0.688 -0.6004>, rot=<0.0303 -0.0457 0.0678 -0.9962>}
 dyn_ribbon_tape_chest_B_l_2 {parent=dyn_ribbon_tape_chest_B_l_1, loc=<2.6091 0 0>, rot=<0 0 0 1>}
 base_ribbon_chest_A {parent=spine_04, loc=<3.119 12.9918 0>, rot=<0.5471 0.448 0.5471 -0.448>}
 dyn_main_ribbon_chest_A {parent=base_ribbon_chest_A, loc=<0 0 0>, rot=<0 0 0 1>}
 base_ribbon_chest_tape_A_r {parent=dyn_main_ribbon_chest_A, loc=<0.6442 -1.0986 0.2432>, rot=<0.0442 0.0442 0.7057 -0.7057>}
 dyn_ribbon_chest_tape_A_r {parent=base_ribbon_chest_tape_A_r, loc=<0 0 0>, rot=<0 0 0 1>}
 dyn_ribbon_chest_a_r {parent=dyn_main_ribbon_chest_A, loc=<0.8146 0 0>, rot=<0.0316 0 0 -0.9995>}
 dyn_ribbon_chest_a_l {parent=dyn_main_ribbon_chest_A, loc=<-0.669 0 0>, rot=<0.0316 0 0 -0.9995>}
 base_ribbon_chest_tape_A_l {parent=dyn_main_ribbon_chest_A, loc=<-0.7634 -0.7317 0.0928>, rot=<0 0 0.7071 -0.7071>}
 dyn_ribbon_chest_tape_A_l {parent=base_ribbon_chest_tape_A_l, loc=<0 0 0>, rot=<0 0 0 1>}
 dyn_chest_l_base {parent=spine_04, loc=<0.5408 4.9573 -5.2979>, rot=<-0.4316 -0.5601 -0.4767 -0.5223>}
 dyn_chest_l {parent=dyn_chest_l_base, loc=<0 0 0>, rot=<0 0 0 1>}
 dyn_chest_r_base {parent=spine_04, loc=<0.5408 4.9573 5.298>, rot=<0.501 0.499 0.5408 0.4556>}
 dyn_chest_r {parent=dyn_chest_r_base, loc=<0 0 0>, rot=<0 0 0 1>}
 base_side_shirt_l {parent=spine_03, loc=<0.6917 1.0723 -11.026>, rot=<0.0521 -0.9886 0.0022 -0.1412>}
 dyn_side_shirt_l {parent=base_side_shirt_l, loc=<0 0 0>, rot=<0 0 0 1>}
 base_bk_shirt_l {parent=spine_03, loc=<-0.555 -5.685 -6.6305>, rot=<-0.02 -0.8167 0.5653 -0.1143>}
 dyn_bk_shirt_l {parent=base_bk_shirt_l, loc=<0 0 0>, rot=<0 0 0 1>}
 base_front_shirt_r {parent=spine_03, loc=<0.0144 9.6151 6.4679>, rot=<0.0674 0.4824 0.8733 0.0082>}
 dyn_front_shirt_r {parent=base_front_shirt_r, loc=<0 0 0>, rot=<0 0 0 1>}
 base_front_shirt_l {parent=spine_03, loc=<0.0143 9.6151 -6.4679>, rot=<-0.0082 -0.8733 -0.4824 -0.0674>}
 dyn_front_shirt_l {parent=base_front_shirt_l, loc=<0 0 0>, rot=<0 0 0 1>}
 base_bk_shirt_r {parent=spine_03, loc=<-0.5547 -5.685 6.6305>, rot=<0.1143 -0.5653 0.8167 0.02>}
 dyn_bk_shirt_r {parent=base_bk_shirt_r, loc=<0 0 0>, rot=<0 0 0 1>}
 base_side_shirt_r {parent=spine_03, loc=<0.6916 1.0723 11.026>, rot=<0.1412 -0.0022 0.9886 -0.0521>}
 dyn_side_shirt_r {parent=base_side_shirt_r, loc=<0 0 0>, rot=<0 0 0 1>}
 base_bk_shirt_c {parent=spine_03, loc=<-2.7299 -6.3051 0.0296>, rot=<0.105 -0.7179 0.681 -0.0996>}
 dyn_bk_shirt_c {parent=base_bk_shirt_c, loc=<0 0 0>, rot=<0 0 0 1>}
 thigh_l {parent=pelvis, loc=<-1.0855 -0.2109 -8.7053>, rot=<0.0009 0.0355 0.0155 -0.9993>}
 calf_l {parent=thigh_l, loc=<-42.5284 -0.0019 -0.7297>, rot=<0.0006 -0.0161 -0.0391 -0.9991>}
 foot_l {parent=calf_l, loc=<-46.7401 -2.5014 -0.592>, rot=<0.0427 -0.0238 0.0353 -0.9982>}
 ball_l {parent=foot_l, loc=<-3.5658 10.8066 -0.1544>, rot=<0.0001 -0.0001 -0.7063 -0.7079>}
 base_laces_foot_l {parent=foot_l, loc=<-0.2025 8.7833 0.2445>, rot=<0.6863 0.1944 0.6801 -0.1693>}
 dyn_main_laces_foot_l {parent=base_laces_foot_l, loc=<0 0 0>, rot=<0 0 0 1>}
 dyn_laces_foot_l_r {parent=dyn_main_laces_foot_l, loc=<0.8146 0 0>, rot=<0 0 1 0>}
 dyn_laces_foot_l_l {parent=dyn_main_laces_foot_l, loc=<-0.669 0 0>, rot=<0 0 1 0>}
 calf_twist_02_l {parent=calf_l, loc=<-19.1546 -1.7943 -0.2638>, rot=<0.0237 0.0095 -0.0035 -0.9997>}
 calf_twist_01_l {parent=calf_l, loc=<-35.8483 -1.9197 -0.5794>, rot=<0.0237 0.0095 -0.0035 -0.9997>}
 deform_knee_in_base_l {parent=calf_l, loc=<0.8031 -5.0231 0.2456>, rot=<0 0 0 -1>}
 deform_knee_in_l {parent=deform_knee_in_base_l, loc=<0 0 0>, rot=<0 0 0 -1>}
 deform_knee_out_base_l {parent=calf_l, loc=<0 4 0>, rot=<0 0 0 -1>}
 deform_knee_out_l {parent=deform_knee_out_base_l, loc=<0 0 0>, rot=<0 0 0 1>}
 thigh_twist_01_l {parent=thigh_l, loc=<-27.952 0 0>, rot=<0.0013 0 0 -1>}
 base_ribbon_leg_l {parent=thigh_twist_01_l, loc=<0.0173 8.3761 -0.6222>, rot=<0.486 0.5524 0.4367 -0.5176>}
 dyn_main_ribbon_leg_l {parent=base_ribbon_leg_l, loc=<0 0 0>, rot=<0 0 0 1>}
 base_ribbon_tape_leg_l {parent=dyn_main_ribbon_leg_l, loc=<0.1412 -1.0255 0.0343>, rot=<0 0 0.7071 -0.7071>}
 dyn_ribbon_tape_leg_l {parent=base_ribbon_tape_leg_l, loc=<0 0 0>, rot=<0 0 0 1>}
 dyn_ribbon_leg_l_l {parent=dyn_main_ribbon_leg_l, loc=<-0.669 0 0>, rot=<0 0 0 1>}
 dyn_ribbon_leg_l_r {parent=dyn_main_ribbon_leg_l, loc=<0.8146 0 0>, rot=<0 0 0 1>}
 deform_groin_base_l {parent=thigh_l, loc=<0.4801 5.6776 -0.2486>, rot=<0 0 0 -1>}
 deform_groin_l {parent=deform_groin_base_l, loc=<0 0 0>, rot=<0 0 0 -1>}
 base_front_skirt_l {parent=thigh_l, loc=<-7.2944 8.2883 0.1142>, rot=<-0.0174 -0.7506 -0.6598 0.0315>}
 dyn_front_skirt_l {parent=base_front_skirt_l, loc=<0 0 0>, rot=<0 0 0 1>}
 dyn_flap_front_skirt_l {parent=base_front_skirt_l, loc=<-0.627 -4.7036 -0.8875>, rot=<-0.1596 0.5632 -0.2084 -0.7835>}
 base_side_skirt_l {parent=thigh_l, loc=<-6.7162 -0.4211 -7.4865>, rot=<-0.0041 -0.9943 0.01 -0.1056>}
 dyn_side_skirt_l {parent=base_side_skirt_l, loc=<0 0 0>, rot=<0 0 0 1>}
 base_bk_skirt_l {parent=thigh_l, loc=<-4.0753 -11.9841 0.9884>, rot=<0.0178 -0.7672 0.6405 0.0304>}
 dyn_bk_skirt_l {parent=base_bk_skirt_l, loc=<0 0 0>, rot=<0 0 0 1>}
 thigh_r {parent=pelvis, loc=<-1.0855 -0.2109 8.7053>, rot=<0.0355 -0.0009 0.9993 0.0155>}
 calf_r {parent=thigh_r, loc=<42.5284 0.0019 0.7297>, rot=<0.0006 -0.0161 -0.0391 -0.9991>}
 foot_r {parent=calf_r, loc=<46.7401 2.5014 0.592>, rot=<0.0427 -0.0238 0.0353 -0.9982>}
 ball_r {parent=foot_r, loc=<3.5658 -10.8066 0.1544>, rot=<0.0001 -0.0001 -0.7063 -0.7079>}
 base_laces_foot_r {parent=foot_r, loc=<0.2025 -8.7833 -0.2445>, rot=<0.605 -0.3434 0.6183 0.3656>}
 dyn_main_laces_foot_r {parent=base_laces_foot_r, loc=<0 0 0>, rot=<0 0 0 1>}
 dyn_laces_foot_r_l {parent=dyn_main_laces_foot_r, loc=<-0.8147 0 0>, rot=<0.7071 0 0 -0.7071>}
 dyn_laces_foot_r_r {parent=dyn_main_laces_foot_r, loc=<0.6689 0 0>, rot=<0.7071 0 0 -0.7071>}
 calf_twist_02_r {parent=calf_r, loc=<19.1546 1.7943 0.2638>, rot=<0.0237 0.0095 -0.0035 -0.9997>}
 calf_twist_01_r {parent=calf_r, loc=<35.8483 1.9197 0.5794>, rot=<0.0237 0.0095 -0.0035 -0.9997>}
 deform_knee_in_base_r {parent=calf_r, loc=<-0.8031 5.0231 -0.2457>, rot=<0 0 0 -1>}
 deform_knee_in_r {parent=deform_knee_in_base_r, loc=<0 0 0>, rot=<0 0 0 -1>}
 deform_knee_out_base_r {parent=calf_r, loc=<0 -4 0>, rot=<0 0 0 -1>}
 deform_knee_out_r {parent=deform_knee_out_base_r, loc=<0 0 0>, rot=<0 0 0 1>}
 thigh_twist_01_r {parent=thigh_r, loc=<27.952 0 0>, rot=<0.0013 0 0 -1>}
 deform_groin_base_r {parent=thigh_r, loc=<-0.4801 -5.6776 0.2486>, rot=<0 0 0 -1>}
 deform_groin_r {parent=deform_groin_base_r, loc=<0 0 0>, rot=<0 0 0 -1>}
 base_front_skirt_r {parent=thigh_r, loc=<7.2944 -8.2883 -0.1142>, rot=<-0.6598 -0.0315 0.0174 -0.7506>}
 dyn_front_skirt_r {parent=base_front_skirt_r, loc=<0 0 0>, rot=<0 0 0 1>}
 base_side_skirt_r {parent=thigh_r, loc=<6.7161 0.4211 7.4866>, rot=<0.01 0.1056 0.0041 -0.9943>}
 dyn_side_skirt_r {parent=base_side_skirt_r, loc=<0 0 0>, rot=<0 0 0 1>}
 base_bk_skirt_r {parent=thigh_r, loc=<4.0819 12.3055 -0.9886>, rot=<0.6405 -0.0304 -0.0178 -0.7672>}
 dyn_bk_skirt_r {parent=base_bk_skirt_r, loc=<0 0 0>, rot=<0 0 0 1>}
 deform_glute_base_l {parent=pelvis, loc=<-7.3466 -9.0213 -9.1245>, rot=<0.0172 0.031 -0.4477 -0.8935>}
 deform_glute_l {parent=deform_glute_base_l, loc=<0 0 0>, rot=<0 0 0 1>}
 deform_glute_base_r {parent=pelvis, loc=<-7.3464 -9.0213 9.1245>, rot=<0.031 -0.0172 0.8935 -0.4477>}
 deform_glute_r {parent=deform_glute_base_r, loc=<0 0 0>, rot=<0 0 0 1>}
 base_ribbon_pelvis {parent=pelvis, loc=<4.6109 5.9109 -10.2772>, rot=<-0.2021 -0.8432 -0.4895 -0.0928>}
 dyn_ribbon_main_pelvis {parent=base_ribbon_pelvis, loc=<0 0 0>, rot=<0 0 0 1>}
 dyn_ribbon_pelvis_l {parent=dyn_ribbon_main_pelvis, loc=<0 -0.8385 0>, rot=<0 0 0.7071 -0.7071>}
 dyn_ribbon_pelvis_r {parent=dyn_ribbon_main_pelvis, loc=<0 0.8352 0>, rot=<0 0 -0.7071 -0.7071>}
 base_ribbon_tape_pelvis {parent=dyn_ribbon_main_pelvis, loc=<1.0202 0 0>, rot=<0 0.0556 0 -0.9985>}
 dyn_ribbon_tape_pelvis_r_1 {parent=base_ribbon_tape_pelvis, loc=<0.2142 1.6743 -0.841>, rot=<0.0122 -0.0707 -0.0277 -0.997>}
 dyn_ribbon_tape_pelvis_r_2 {parent=dyn_ribbon_tape_pelvis_r_1, loc=<4.8599 0 0>, rot=<0 0 0 1>}
 dyn_ribbon_tape_pelvis_l_1 {parent=base_ribbon_tape_pelvis, loc=<0.2276 -1.0043 -0.5791>, rot=<0.0303 -0.0457 0.0678 -0.9962>}
 dyn_ribbon_tape_pelvis_l_2 {parent=dyn_ribbon_tape_pelvis_l_1, loc=<3.997 0 0>, rot=<0 0 0 1>}
]
variables [playingEmote {type=bool, default=False}]
