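# Physics asset for the temporary TangyRadishFlame female medium player body, covering two ribbon tape chains at the pelvis.
# A small base ribbon body joins the pelvis with free rotation and anchors left and right two-segment capsule chains through swing-limited, nearly twist-locked constraints.
# Larger capsules on the pelvis and left thigh provide body shapes around the ribbons; the thigh has no constraint and the right thigh has no body.
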PhysicsAsset /BRCosmetics/Characters/Player/Female/Medium/Bodies/F_Med_TangyRadishFlame_Temp/F_MED_TangyRadishFlame_Physics_TEMP
bodies {
 base_ribbon_tape_pelvis [{shapeType=Capsule, length=0.505, radius=0.505, shapeName=base_ribbon_tape_pelvis_capsule, rotation=<rot 0 -90 0>}]
 dyn_ribbon_tape_pelvis_l_1 [{shapeType=Capsule, length=5, radius=2, center=<3 0 2.5>, rotation=<rot -90 0 0>}]
 dyn_ribbon_tape_pelvis_l_2 [{shapeType=Capsule, length=5, radius=2.3, center=<3 0 2.5>, rotation=<rot -90 0 0>}]
 dyn_ribbon_tape_pelvis_r_1 [{shapeType=Capsule, length=5, radius=2, center=<3 0 2.5>, rotation=<rot -90 0 0>}]
 dyn_ribbon_tape_pelvis_r_2 [{shapeType=Capsule, length=5, radius=2.3, center=<3 0 2.5>, rotation=<rot -90 0 0>}]
 pelvis [{shapeType=Capsule, length=6, radius=9, center=<0.7858 0 0>, rotation=<rot 0 -0.0001 0>}]
 thigh_l [
  {shapeType=Capsule, length=5.5, radius=9, center=<-6.5219 1.2408 -0.0767>, rotation=<rot 0 -0.0007 15>}
  {shapeType=Capsule, length=5.5, radius=9, center=<-13.1628 1.2409 -0.0767>, rotation=<rot 0 -0.0007 15>}
 ]
}
constraints [
 {
  bone1Name=dyn_ribbon_tape_pelvis_l_2
  bone2Name=dyn_ribbon_tape_pelvis_l_1
  swing1LimitDegrees=25
  swing2LimitDegrees=45
  twistLimitDegrees=2
 }
 {
  bone1Name=dyn_ribbon_tape_pelvis_r_2
  bone2Name=dyn_ribbon_tape_pelvis_r_1
  swing1LimitDegrees=25
  swing2LimitDegrees=45
  twistLimitDegrees=2
 }
 {bone1Name=dyn_ribbon_tape_pelvis_r_1, bone2Name=base_ribbon_tape_pelvis, swing1LimitDegrees=25, swing2LimitDegrees=45, twistLimitDegrees=2}
 {bone1Name=dyn_ribbon_tape_pelvis_l_1, bone2Name=base_ribbon_tape_pelvis, swing1LimitDegrees=25, swing2LimitDegrees=45, twistLimitDegrees=2}
 {
  bone1Name=base_ribbon_tape_pelvis
  bone2Name=pelvis
  swing1LimitDegrees=45
  swing1Motion=Free
  swing2LimitDegrees=45
  swing2Motion=Free
  twistLimitDegrees=45
  twistMotion=Free
 }
]
